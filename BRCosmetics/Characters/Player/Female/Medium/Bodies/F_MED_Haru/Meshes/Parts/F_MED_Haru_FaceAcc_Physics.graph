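# Physics asset for the face accessory of the female medium player cosmetic body F_MED_Haru, covering the head and a tuft bone chain.
# Two capsules form the head body, and single capsules cover base_tuft and dyn_tuft_00 through 03.
# Constraints chain the tuft bones back to the head, with base_tuft angularly free against it.
# The first dynamic link has zero angular limits, while later links allow moderate swing without twist.
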PhysicsAsset /BRCosmetics/Characters/Player/Female/Medium/Bodies/F_MED_Haru/Meshes/Parts/F_MED_Haru_FaceAcc_Physics
bodies {
 base_tuft [{shapeType=Capsule, length=0.505, radius=0.505, shapeName=base_tuft_capsule, rotation=<rot 0 -90 0>}]
 dyn_tuft_00 [{shapeType=Capsule, length=2, radius=1, shapeName=dyn_tuft_00_capsule, center=<2 0 0>, rotation=<rot 90 0 0>}]
 dyn_tuft_01 [{shapeType=Capsule, length=5.4342, radius=1, shapeName=dyn_tuft_01_capsule, center=<2.9685 0 0>, rotation=<rot 90 0 0>}]
 dyn_tuft_02 [{shapeType=Capsule, length=3.045, radius=1, shapeName=dyn_tuft_02_capsule, center=<2.2241 0 0>, rotation=<rot 90 0 0>}]
 dyn_tuft_03 [{shapeType=Capsule, length=3.6114, radius=1, shapeName=dyn_tuft_03_capsule, center=<2.486 0 0>, rotation=<rot 90 0 0>}]
 head [
  {shapeType=Capsule, length=5.3985, radius=6.6015, shapeName=base_head_capsule, center=<1 2.2794 0>, rotation=<rot 90 0 0>}
  {shapeType=Capsule, length=3.9786, radius=5.4062, shapeName=head_capsule, center=<8.1994 3.7639 0.6298>, rotation=<rot 0 -0.0001 5>}
 ]
}
constraints [
 {bone1Name=dyn_tuft_00, bone2Name=base_tuft, swing1LimitDegrees=0, swing2LimitDegrees=0, twistLimitDegrees=0}
 {bone1Name=dyn_tuft_01, bone2Name=dyn_tuft_00, swing1LimitDegrees=15, swing2LimitDegrees=35, twistLimitDegrees=0}
 {bone1Name=dyn_tuft_02, bone2Name=dyn_tuft_01, swing1LimitDegrees=25, swing2LimitDegrees=35, twistLimitDegrees=0}
 {bone1Name=dyn_tuft_03, bone2Name=dyn_tuft_02, swing1LimitDegrees=25, swing2LimitDegrees=35, twistLimitDegrees=0}
 {
  bone1Name=base_tuft
  bone2Name=head
  swing1LimitDegrees=45
  swing1Motion=Free
  swing2LimitDegrees=45
  swing2Motion=Free
  twistLimitDegrees=45
  twistMotion=Free
 }
]
